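# Light fog ambience loop cue for the Apollo biomes, played quietly on the outdoor weather sound class with Athena ambient attenuation.
# A quality switch picks a looping quad wave for the first two levels and a looping stereo wave for the last two.
# An Enveloper sways pitch along a roughly two-minute curve at flat volume.
# A root Modulator adds slight random downward pitch variation per play.
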SoundCue /Game/Sounds/Apollo/Biomes/Fog/Biome_Fog_Light_Loop_Cue
root=Modulator_0
nodes [
 {
  id=Enveloper_0
  children [QualityLevel_0]
  properties {
   pitchCurve {editorCurveData {keys [
      {time=0}
      <at 12.4568 0.8806>
      <at 25.8034 1.006>
      <at 43.5988 0.9004>
      <at 64.5084 1.006>
      <at 81.4141 0.9268>
      <at 96.9851 1.006>
      <at 110.7766 0.9378>
      <at 123.6782 1.006>
     ]}}
   volumeCurve {editorCurveData {keys [<at 0 1>]}}
  }
 }
 {id=Modulator_0, children [Enveloper_0], properties {pitchMax=1, pitchMin=0.9, volumeMax=1, volumeMin=1}}
 {id=QualityLevel_0, children [WavePlayer_2,WavePlayer_2,WavePlayer_3,WavePlayer_3]}
 {id=WavePlayer_2, properties {bLooping=true, soundWaveAssetPtr=&/Amb_Fog_Loop_Quad}}
 {id=WavePlayer_3, properties {bLooping=true, soundWaveAssetPtr=&/Amb_Fog_Loop_ST}}
]
properties {
 attenuationSettings=&Fort_Ambient_Athena_Att
 maxDistance=2097152
 soundClassObject=/Game/Sounds/FortSoundClassesAndModes/Fort_Ambient_Outdoors_Weather_Athena
 volumeMultiplier=0.1
 preEffectBusSends [{}]
}
