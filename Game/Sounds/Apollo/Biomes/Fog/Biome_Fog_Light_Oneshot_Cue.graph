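# Sound Cue for one-shot light fog biome ambience, routed to the outdoor weather sound class with tile-ambient attenuation.
# A mixer layers three pitch- and volume-modulated random picks: standard fog one-shots, plus two quieter Heavy sets.
# Empty random slots let both Heavy layers sometimes stay silent for variation.
# The quality node plays the mix only on its first two levels, and a concurrency override prevents overlapping instances.
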SoundCue /Game/Sounds/Apollo/Biomes/Fog/Biome_Fog_Light_Oneshot_Cue
root=QualityLevel_0
nodes [
 {id=Mixer_1, children [Modulator_1,Modulator_0,Modulator_2], properties {inputVolume [1,0.85,0.15]}}
 {id=Modulator_0, children [Random_2], properties {pitchMax=1.2, pitchMin=0.6, volumeMax=1, volumeMin=0.7}}
 {id=Modulator_1, children [Random_1], properties {pitchMax=1, pitchMin=0.6, volumeMax=1, volumeMin=0.9}}
 {id=Modulator_2, children [Random_3], properties {pitchMax=1.1, pitchMin=0.7, volumeMax=1, volumeMin=1}}
 {id=QualityLevel_0, children [Mixer_1,Mixer_1,null,null]}
 {id=Random_1, children [WavePlayer_0,WavePlayer_7,WavePlayer_8], properties {weights [1,1,1]}}
 {id=Random_2, children [WavePlayer_1,WavePlayer_2,null], properties {weights [1,1,1]}}
 {
  id=Random_3
  children [WavePlayer_9, WavePlayer_10, WavePlayer_11, WavePlayer_12, WavePlayer_13, null, null, null, null]
  properties {weights [1,1,1,1,1,1,1,1,1]}
 }
 {id=WavePlayer_0, properties {soundWaveAssetPtr=&/Amb_Fog_Oneshot_01}}
 {id=WavePlayer_1, properties {soundWaveAssetPtr=&/Amb_Fog_Oneshot_Heavy_01}}
 {id=WavePlayer_10, properties {soundWaveAssetPtr=&/Amb_Fog_Oneshot_Heavy_09}}
 {id=WavePlayer_11, properties {soundWaveAssetPtr=&/Amb_Fog_Oneshot_Heavy_11}}
 {id=WavePlayer_12, properties {soundWaveAssetPtr=&/Amb_Fog_Oneshot_Heavy_12}}
 {id=WavePlayer_13, properties {soundWaveAssetPtr=&/Amb_Fog_Oneshot_Heavy_06}}
 {id=WavePlayer_2, properties {soundWaveAssetPtr=&/Amb_Fog_Oneshot_Heavy_02}}
 {id=WavePlayer_7, properties {soundWaveAssetPtr=&/Amb_Fog_Oneshot_02}}
 {id=WavePlayer_8, properties {soundWaveAssetPtr=&/Amb_Fog_Oneshot_03}}
 {id=WavePlayer_9, properties {soundWaveAssetPtr=&/Amb_Fog_Oneshot_Heavy_07}}
]
properties {
 attenuationSettings=/Game/Sounds/Attenuation/Fort_TileAmbient_OneShot_Att
 bOverrideConcurrency=true
 maxDistance=8000
 soundClassObject=/Game/Sounds/FortSoundClassesAndModes/Fort_Ambient_Outdoors_Weather_Athena
 volumeMultiplier=0.351
 concurrencyOverrides {maxCount=1, resolutionRule=PreventNew}
 preEffectBusSends [{}]
}
