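# Sound cue for medium fog biome ambient one-shots, in the outdoor weather sound class with tile-ambient one-shot attenuation.
# A mixer layers three randomized pools, one regular and two heavy, each with its own pitch and volume modulation.
# The third pool includes two empty entries among its equally weighted choices, so it sometimes adds nothing.
# The first two quality levels share this mixer, the others are empty, and concurrency allows one instance, blocking new plays.
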SoundCue /Game/Sounds/Apollo/Biomes/Fog/Biome_Fog_Medium_Oneshot_Cue
root=QualityLevel_0
nodes [
 {id=Mixer_1, children [Modulator_2,Modulator_3,Modulator_4], properties {inputVolume [1,0.95,0.3]}}
 {id=Modulator_2, children [Random_2], properties {pitchMax=0.8, pitchMin=0.6, volumeMax=1.1, volumeMin=0.9}}
 {id=Modulator_3, children [Random_3], properties {pitchMax=1.2, pitchMin=0.6, volumeMax=1, volumeMin=0.5}}
 {id=Modulator_4, children [Random_4], properties {pitchMax=1.1, pitchMin=0.7, volumeMax=1, volumeMin=1}}
 {id=QualityLevel_0, children [Mixer_1,Mixer_1,null,null]}
 {id=Random_2, children [WavePlayer_9, WavePlayer_10, WavePlayer_11], properties {weights [1,1,1]}}
 {id=Random_3, children [WavePlayer_12, WavePlayer_13], properties {weights [1,1]}}
 {
  id=Random_4
  children [WavePlayer_14, WavePlayer_15, WavePlayer_16, WavePlayer_17, WavePlayer_18, null, null]
  properties {weights [1,1,1,1,1,1,1]}
 }
 {id=WavePlayer_10, properties {soundWaveAssetPtr=&/Amb_Fog_Oneshot_02}}
 {id=WavePlayer_11, properties {soundWaveAssetPtr=&/Amb_Fog_Oneshot_03}}
 {id=WavePlayer_12, properties {soundWaveAssetPtr=&/Amb_Fog_Oneshot_Heavy_01}}
 {id=WavePlayer_13, properties {soundWaveAssetPtr=&/Amb_Fog_Oneshot_Heavy_02}}
 {id=WavePlayer_14, properties {soundWaveAssetPtr=&/Amb_Fog_Oneshot_Heavy_07}}
 {id=WavePlayer_15, properties {soundWaveAssetPtr=&/Amb_Fog_Oneshot_Heavy_09}}
 {id=WavePlayer_16, properties {soundWaveAssetPtr=&/Amb_Fog_Oneshot_Heavy_11}}
 {id=WavePlayer_17, properties {soundWaveAssetPtr=&/Amb_Fog_Oneshot_Heavy_12}}
 {id=WavePlayer_18, properties {soundWaveAssetPtr=&/Amb_Fog_Oneshot_Heavy_06}}
 {id=WavePlayer_9, properties {soundWaveAssetPtr=&/Amb_Fog_Oneshot_01}}
]
properties {
 attenuationSettings=/Game/Sounds/Attenuation/Fort_TileAmbient_OneShot_Att
 bOverrideConcurrency=true
 maxDistance=8000
 soundClassObject=/Game/Sounds/FortSoundClassesAndModes/Fort_Ambient_Outdoors_Weather_Athena
 volumeMultiplier=0.326
 concurrencyOverrides {maxCount=1, resolutionRule=PreventNew}
 preEffectBusSends [{}]
}
